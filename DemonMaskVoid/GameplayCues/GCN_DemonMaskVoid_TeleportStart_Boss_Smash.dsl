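; Burst GameplayCue for the start of the Demon Mask Void teleport, Boss Smash variant, setting the particle Color Variant by whether the target's root component name contains "Smash".
; A mesh-name check detects the Wandering Boss, which gets extra trail and start-hole effects, unless a sentinel cue parameter removes the burst particle.
; Other locally viewed pawns get an audio bus mix and sound indicator, toggleable by hotfix.
; Higher-detail non-mobile platforms also spawn a radial force field.
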(fn OnBurst(EitherVFX-preferthisversion)
  (params (MyTarget :type Actor%20Object%20Reference) (Parameters :type Gameplay%20Cue%20Parameters%20Structure%20%28by%20ref%29) (ParticleComponents :type Array%20of%20FXSystem%20Component%20Object%20References) (AudioComponents :type Array%20of%20Audio%20Component%20Object%20References) (BurstCameraShakeInstance :type Camera%20Shake%20Base%20Object%20Reference) (BurstDecalInstance :type Decal%20Actor%20Object%20Reference))
  (bind _fortplayerpawn (LocalVariables|GetFortPlayerPawn))
  (bind _mesh (Variables|/Script/Engine.Character|GetMesh _fortplayerpawn))
  (bind _returnvalue_location (Math|Transform|BreakTransform (Transformation|GetBoneTransform _mesh "None" "RTS_World")))
  (bind (_returnvalue_location_2 _returnvalue_rotation) (Math|Transform|BreakTransform (Transformation|GetBoneTransform (Variables|/Script/Engine.Character|GetMesh (LocalVariables|GetFortPlayerPawn)) "None" "RTS_World")))
  (bind _mytarget (Variables|GetMyTarget))
  (bind _instigator (Variables|/Script/Engine.Actor|GetInstigator _mytarget))
  (bind _particlecomponent0 (LocalVariables|GetParticleComponent0))
  (bind _workonwanderingboss (LocalVariables|GetWorkOnWanderingBoss))
  (|CallParentFunction MyTarget Parameters ParticleComponents AudioComponents BurstCameraShakeInstance BurstDecalInstance)
  ; for Nana Eternal Boss.. I know it's not good way tho..
  (Effects|Components|ParticleSystem|SetIntParameter (Utilities|Array|Get(acopy) ParticleComponents) "Color Variant" (Math|Conversions|ToInteger(Boolean) (Utilities|String|Contains (Utilities|GetDisplayName (Variables|/Script/Engine.Actor|GetRootComponent MyTarget)) "Smash")))
  (Utilities|FlowControl|Sequence
    (:then_0
      ; Setup variables
      (LocalVariables|SetParticleComponent0 (Utilities|Array|Get(acopy) ParticleComponents))
      (bind _asfort_player_pawn (Utilities|Casting|CastToFortPlayerPawn MyTarget))
      (LocalVariables|SetFortPlayerPawn _asfort_player_pawn)
      (LocalVariables|SetWorkOnWanderingBoss (>= (Utilities|String|FindSubstring (Utilities|GetDisplayName (Variables|/Script/Engine.Character|GetMesh (LocalVariables|GetFortPlayerPawn))) "WanderingBoss") 0)))
    (:then_1
      ;[0.03358 0.09832 0.15 0.5] Audio
      (if (not (LocalVariables|GetWorkOnWanderingBoss))
        ;;[0.9105 0 1 0.5] This doesn't need work on wandering boss :)
        (if (Pawn|IsLocallyViewed _instigator)
          (if (LocalVariables|GetWorkOnWanderingBoss)
            _
            (elif (ScalableFloat|GetValueasBool (LocalVariables|GetHF_DisableControlBusMixCall))
              _
              (else
                (Audio|Modulation|ActivateControlBusMix _instigator (Variables|Default|GetCBM_Teleport))
                (CallFunction|PlaySoundIndicator (Transformation|GetActorLocation (Variables|GetMyTarget)) (Variables|/Script/Engine.Actor|GetInstigator (Variables|GetMyTarget)))))))))
    (:then_2
      (Utilities|Casting|CastToFortPlayerPawn MyTarget)
      ;[0.15 0.02517 0.002092 0.5] FX
      ;[0.1362 0 0.15 0.5] This part work on wandering boss.
      (if (== (Ability|GameplayCue|BreakGameplayCueParameters (Variables|GetParameters)) -1e+01)
        (Components|Activation|Deactivate _particlecomponent0)
        (Components|DestroyComponent _particlecomponent0 0)
        (elif (LocalVariables|GetWorkOnWanderingBoss)
          (Utilities|FlowControl|Sequence
            (:then_0
              (Development|PrintString 0 "VoidMaskWorkingOnWanderingBoss")
              (bind _returnvalue (Niagara|SpawnSystemAttached "/DemonMaskVoid/Assets/VFX/NS_WandBoss_VoidMask_Teleport_WrapTrail.NS_WandBoss_VoidMask_Teleport_WrapTrail" _mesh "pelvis" "0, 0, 0" "0, 0, 0" "SnapToTarget" true))
              (Niagara|SetNiagaraVariable(Position) _returnvalue "OriginPos" _returnvalue_location)
              (Effects|Components|ParticleSystem|SetIntParameter _returnvalue "Color Variant" (Variables|Default|GetColor%20Variant)))
            (:then_1
              ;; This part disabled for player void mask ,but we would like let it works on wandering boss.
              (bind _returnvalue_2 (Effects|Components|ParticleSystem|SpawnFXAtLocation 0 "/DemonMaskVoid/Assets/VFX/NS_VoidMask_Teleport_StartHole.NS_VoidMask_Teleport_StartHole" _returnvalue_location_2 _returnvalue_rotation))
              (Effects|Components|ParticleSystem|SetIntParameter _returnvalue_2 "Color Variant" (Variables|Default|GetColor%20Variant))
              (Effects|Components|ParticleSystem|SetFloatParameter _returnvalue_2 "Global Scale" (Math|Float|SelectFloat 2.0 1.0 (LocalVariables|GetWorkOnWanderingBoss)))
              (Effects|Components|ParticleSystem|SetBoolParameter (LocalVariables|GetParticleComponent0) "WanderingBoss" _workonwanderingboss)
              (Effects|Components|ParticleSystem|SetVectorParameter _returnvalue_2 "Offset Pos" (Math|Vector|SelectVector "0.000000,20.000000,340.000000" "0.000000,20.000000,180.000000" _workonwanderingboss)))))))
    (:then_3
      ;[0.1302 0 0] VFX: Radial Force
      (if (Math|Boolean|ANDBoolean (not (Fortnite|IsMobilePlatform)) (not (Class|PlatformHelpersLibrary|IsPcPerfMode 0)) (> (Rendering|GetRenderingDetailMode) 1))
        (Game|SpawnActorfromClass 0 "/Game/Effects/Niagara/Blueprints/PhsyicsFields/FN_RadialForce.FN_RadialForce_C" (Math|Transform|MakeTransform (Transformation|GetActorLocation (Variables|GetMyTarget))) "Undefined" "MultiplyWithRoot" 0 3.0 1.0 8.5e+02 -0.9 0.0 0)))
    (:then_4
      (return))))

(fn PlaySoundIndicator
  (params (Location :type Vector) (Instigator :type Actor%20Object%20Reference))
  (bind _bursteffectdata (Variables|BurstEffects|GetBurstEffectData))
  (bind (_burstparticles _burstsounds) (Utilities|Struct|BreakFortBurstEffectData _bursteffectdata))
  (bind (_condition _attachment _soundcue) (Utilities|Struct|BreakFortGameplayCueAudioInfo (Utilities|Array|Get(aref) _burstsounds)))
  (if (> (Utilities|Array|Length _burstsounds) 0)
    (Fortnite|ShowSoundIndicatorLocal :Location Location :MaxAudibleDistance (Variables|/Script/Engine.SoundBase|GetMaxDistance _soundcue) :IndicatorType "Gunshot" :Instigator Instigator :Tint "(R=0.000000,G=0.000000,B=0.000000,A=0.000000)" :SoundBase _soundcue :AttenuationMaxRangeMultiplier 1.0 :AttenuationIntensityMultiplier 1.0) ; Indicator for fuse
  ))
