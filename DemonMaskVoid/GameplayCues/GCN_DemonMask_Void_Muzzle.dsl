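; Burst gameplay cue notify for the muzzle effect of the Demon Mask Void item.
; OnBurst runs the parent burst, then for a valid player pawn target passes the normalized base aim direction to the first particle component as the "AimVec" vector parameter.
; PlaySoundIndicator shows a local "Gunshot" sound indicator at a given location, using the first burst sound and its max distance when burst sounds exist.
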(fn OnBurst(EitherVFX-preferthisversion)
  (params (MyTarget :type Actor%20Object%20Reference) (Parameters :type Gameplay%20Cue%20Parameters%20Structure%20%28by%20ref%29) (ParticleComponents :type Array%20of%20FXSystem%20Component%20Object%20References) (AudioComponents :type Array%20of%20Audio%20Component%20Object%20References) (BurstCameraShakeInstance :type Camera%20Shake%20Base%20Object%20Reference) (BurstDecalInstance :type Decal%20Actor%20Object%20Reference))
  (|CallParentFunction MyTarget Parameters ParticleComponents AudioComponents BurstCameraShakeInstance BurstDecalInstance)
  (Utilities|IsValid MyTarget
    (:Is%20Valid
      (bind _asplayer_pawn_athena (Utilities|Casting|CastToPlayerPawn_Athena MyTarget))
      (Effects|Components|ParticleSystem|SetVectorParameter (Utilities|Array|Get(acopy) ParticleComponents) "AimVec" (Math|Vector|Normalize (Math|Rotator|GetRotationXVector (Pawn|GetBaseAimRotation _asplayer_pawn_athena))))
      (return))))

(fn PlaySoundIndicator
  (params (Location :type Vector) (Instigator :type Actor%20Object%20Reference))
  (bind _bursteffectdata (Variables|BurstEffects|GetBurstEffectData))
  (bind (_burstparticles _burstsounds) (Utilities|Struct|BreakFortBurstEffectData _bursteffectdata))
  (bind (_condition _attachment _soundcue) (Utilities|Struct|BreakFortGameplayCueAudioInfo (Utilities|Array|Get(aref) _burstsounds)))
  (if (> (Utilities|Array|Length _burstsounds) 0)
    (Fortnite|ShowSoundIndicatorLocal :Location Location :MaxAudibleDistance (Variables|/Script/Engine.SoundBase|GetMaxDistance _soundcue) :IndicatorType "Gunshot" :Instigator Instigator :Tint "(R=0.000000,G=0.000000,B=0.000000,A=0.000000)" :SoundBase _soundcue :AttenuationMaxRangeMultiplier 1.0 :AttenuationIntensityMultiplier 1.0) ; Indicator for fuse
  ))
