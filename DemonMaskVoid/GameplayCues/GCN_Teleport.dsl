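; Gameplay cue notify that plays burst effects for the Demon Mask Void teleport.
; After the parent burst, a sentinel cue parameter destroys the first particle component; otherwise that component is scaled and flagged by searching the mesh name for "WanderingBoss".
; Bosses also get pelvis attachment and extra Niagara smoke, while higher-detail non-mobile platforms spawn a radial force field.
; A separate function shows a local sound indicator from the first burst sound.
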(fn OnBurst(EitherVFX-preferthisversion)
  (params (MyTarget :type Actor%20Object%20Reference) (Parameters :type Gameplay%20Cue%20Parameters%20Structure%20%28by%20ref%29) (ParticleComponents :type Array%20of%20FXSystem%20Component%20Object%20References) (AudioComponents :type Array%20of%20Audio%20Component%20Object%20References) (BurstCameraShakeInstance :type Camera%20Shake%20Base%20Object%20Reference) (BurstDecalInstance :type Decal%20Actor%20Object%20Reference))
  (bind _output (Utilities|Array|Get(acopy) ParticleComponents))
  (|CallParentFunction MyTarget Parameters ParticleComponents AudioComponents BurstCameraShakeInstance BurstDecalInstance)
  (if (== (Ability|GameplayCue|BreakGameplayCueParameters Parameters) -1e+01)
    (Components|Activation|Deactivate _output)
    (Components|DestroyComponent _output 0)
    (else
      (Utilities|FlowControl|Sequence
        (:then_0
          (bind _asfort_player_pawn (Utilities|Casting|CastToFortPlayerPawn MyTarget)))
        (:then_1
          ;[1 0 0 0.5] VFX:  Need to change Scale when Boss(Pram should be 2.0)
          (Effects|Components|ParticleSystem|SetFloatParameter _output "Global Scale" (Math|Float|SelectFloat 3.0 1.0 (>= (Utilities|String|FindSubstring (Utilities|GetDisplayName (Variables|/Script/Engine.Character|GetMesh _asfort_player_pawn)) "WanderingBoss") 0)))
          (Effects|Components|ParticleSystem|SetBoolParameter _output "WanderingBoss" (>= (Utilities|String|FindSubstring (Utilities|GetDisplayName (Variables|/Script/Engine.Character|GetMesh _asfort_player_pawn)) "WanderingBoss") 0))
          (Utilities|FlowControl|Sequence
            (:then_0
              (if (>= (Utilities|String|FindSubstring (Utilities|GetDisplayName (Variables|/Script/Engine.Character|GetMesh _asfort_player_pawn)) "WanderingBoss") 0)
                (Transformation|AttachComponentToComponent _output (Variables|/Script/Engine.Character|GetMesh _asfort_player_pawn) "pelvis" "KeepRelative" "KeepRelative" "KeepWorld" true)
                (bind _returnvalue (Niagara|SpawnSystemAttached "/DemonMaskVoid/Assets/VFX/NS_WandBoss_VoidMask_Teleport_Smoke.NS_WandBoss_VoidMask_Teleport_Smoke" (Variables|/Script/Engine.Character|GetMesh _asfort_player_pawn) "root" "0, 0, 0" "0, 0, 0" "SnapToTarget" true))
                (Transformation|SetAbsolute _returnvalue false false true)
                (Transformation|SetWorldScale3D _returnvalue "1.000000,1.000000,1.000000")))
            (:then_1
              ;[0.1302 0 0] VFX: Radial Force
              (if (Math|Boolean|ANDBoolean (not (Fortnite|IsMobilePlatform)) (not (Class|PlatformHelpersLibrary|IsPcPerfMode 0)) (> (Rendering|GetRenderingDetailMode) 1))
                (Game|SpawnActorfromClass 0 "/Game/Effects/Niagara/Blueprints/PhsyicsFields/FN_RadialForce.FN_RadialForce_C" (Math|Transform|MakeTransform (Transformation|GetActorLocation MyTarget)) "Undefined" "MultiplyWithRoot" 0 3.0 1.0 8.5e+02 -0.9 0.3 0)
                (return)))))))))

(fn PlaySoundIndicator
  (params (Location :type Vector) (Instigator :type Actor%20Object%20Reference))
  (bind _bursteffectdata (Variables|BurstEffects|GetBurstEffectData))
  (bind (_burstparticles _burstsounds) (Utilities|Struct|BreakFortBurstEffectData _bursteffectdata))
  (bind (_condition _attachment _soundcue) (Utilities|Struct|BreakFortGameplayCueAudioInfo (Utilities|Array|Get(aref) _burstsounds)))
  (if (> (Utilities|Array|Length _burstsounds) 0)
    (Fortnite|ShowSoundIndicatorLocal :Location Location :MaxAudibleDistance (Variables|/Script/Engine.SoundBase|GetMaxDistance _soundcue) :IndicatorType "Gunshot" :Instigator Instigator :Tint "(R=0.000000,G=0.000000,B=0.000000,A=0.000000)" :SoundBase _soundcue :AttenuationMaxRangeMultiplier 1.0 :AttenuationIntensityMultiplier 1.0) ; Indicator for fuse
  ))
